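# Unlit translucent particle material with depth test disabled, drawing a scrolling tunnel for the hiding prop teleport screen effect.
# Offset, mip-biased pipe texture samples average into a streaked blur, combined with fake lighting, cubemap, speed lines, mist and light masks; low-end paths simplify this.
# V gradients shade the tunnel and exit glow, with particle dynamic parameters driving whitewash and brightness.
# A separate branch computes a swirling, distance-faded offset vector.
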Material /FNGameplayCues/Athena/Items/EnvironmentalItems/HidingProps/Proto_HidingProp_Teleport_BlackScreen
graph {
 properties {
  bDisableDepthTest=true
  bUseTranslucencyVertexFog=false
  bUsedWithMeshParticles=true
  blendMode=BLEND_Translucent
  pixelDepthOffsetMode=PDOM_Legacy
  shadingModel=MSM_Unlit
  usedShadingModels=Unlit
 }
 nodes [
  n1 = Transform(None=VectorParameter() {parameterName=FakeLightPos, defaultValue=<rgba 0 0 -0.05 1>}.RGB) {transformSourceType=TRANSFORMSOURCE_World, transformType=TRANSFORM_Tangent}
  n2 = Rotator(
   Coordinate=MaterialFunctionCall!(
    UV Channel=TextureCoordinate()
    U Freq Multiplier=ScalarParameter() {defaultValue=2, group=UVs, parameterName=UV U Mult}
    V Freq Multiplier=ScalarParameter() {defaultValue=3, group=UVs, parameterName=UV V Mult}
    Pan Speed U=ScalarParameter() {defaultValue=0.4, group=UVs, parameterName=UV U Pan Speed}
    Pan Speed V=ScalarParameter() {defaultValue=-1, group=UVs, parameterName=UV V Pan Speed}
   ) {materialFunction=/Game/Effects/Niagara/MasterMaterials/Functions/UVs/MF_FX_TextureUVPanner}.Result
   Time=Constant() {r=4.7125}
  ) {speed=1}
  n3 = TextureSample(UVs=@n2) {
   samplerSource=SSM_Wrap_WorldGroupSettings
   samplerType=SAMPLERTYPE_Normal
   texture=/Game/Environments/Apollo/Sets/Yacht/Textures/T_PipeInterior_N
  }
  n4 = StaticBool()
  n5 = TextureSample!(UVs=Add(A=@n2, B=Constant2Vector() {r=-0.0025}), Tex, Bias=Constant() {r=1}, Apply View MipBias=@n4) {
   mipValueMode=TMVM_MipBias
   samplerSource=SSM_Wrap_WorldGroupSettings
   samplerType=SAMPLERTYPE_Normal
   texture=&T_PipeInterior_N
  }
  n6 = TextureSample(UVs=Add(A=@n2, B=Constant2Vector() {r=-0.005}), Bias=Constant() {r=2}, Apply View MipBias=@n4) {
   mipValueMode=TMVM_MipBias
   samplerSource=SSM_Wrap_WorldGroupSettings
   samplerType=SAMPLERTYPE_Normal
   texture=&T_PipeInterior_N
  }
  n7 = TextureSample(UVs=Add(A=@n2, B=Constant2Vector() {r=-0.0075}), Bias=Constant() {r=2}, Apply View MipBias=@n4) {
   mipValueMode=TMVM_MipBias
   samplerSource=SSM_Wrap_WorldGroupSettings
   samplerType=SAMPLERTYPE_Normal
   texture=&T_PipeInterior_N
  }
  n8 = Add(A=Add(A=Add(A=@n3.RGB, B=@n5.RGB), B=@n6.RGB), B=@n7.RGB)
  n9 = QualitySwitch(Default=ShadingPathSwitch(Default=Divide(A=@n8) {constB=4}, Mobile=@n3.RGB) {desc=Migrated from FeatureLevelSwitch}, Low=@n3.RGB)
  n10 = TextureSample!(UVs=@n2, Tex, Apply View MipBias) {samplerSource=SSM_Wrap_WorldGroupSettings, samplerType=SAMPLERTYPE_Normal, texture=/Game/Global/Textures/Default/Tileable_Patterns/1Normal}
  n11 = MaterialFunctionCall(Normal=@n10.RGB, Flatness=ScalarParameter() {defaultValue=0.925, group=Cubemap, parameterName=Cubemap Distortion Flatness}) {materialFunction=&FlattenNormal}
  n12 = QualitySwitch(
   Default=ShadingPathSwitch(
    Default=MaterialFunctionCall(BaseNormal=@n9, AdditionalNormal=@n11.Result) {materialFunction=&BlendAngleCorrectedNormals}.Result
    Mobile=@n9
   ) {desc=Migrated from FeatureLevelSwitch}
   Low=@n9
  )
  n13 = LinearInterpolate(
   A=VectorParameter() {parameterName=ColorBright, defaultValue=<rgba 0.9167 0.9167 0.9167 1>}.RGB
   B=VectorParameter() {parameterName=ColorDark, defaultValue=<rgba 150 150 150 1>}.RGB
   Alpha=Clamp(None=DotProduct(A=@n1, B=@n12))
  )
  n14 = MaterialFunctionCall(
   UV Channel=TextureCoordinate()
   U Freq Multiplier=ScalarParameter() {defaultValue=2, group=Emissive, parameterName=UV Emissive U Mult}
   V Freq Multiplier=ScalarParameter() {defaultValue=2, group=Emissive, parameterName=UV Emissive V Mult}
   Pan Speed U=ScalarParameter() {defaultValue=0.4, group=Emissive, parameterName=UV Emissive U Pan Speed}
   Pan Speed V=ScalarParameter() {defaultValue=0.4, group=Emissive, parameterName=UV Emissive V Pan Speed}
  ) {materialFunction=&MF_FX_TextureUVPanner}
  n15 = Constant() {r=4.7125}
  n16 = ScalarParameter() {group=Emissive, parameterName=Emissive Offset U}
  n17 = ScalarParameter() {group=Emissive, parameterName=Emissive Offset V}
  n18 = TextureSample(UVs=Add(A=Rotator(Coordinate=@n14.Result, Time=@n15) {speed=1}, B=AppendVector(A=@n16, B=@n17))) {
   samplerSource=SSM_Wrap_WorldGroupSettings
   texture=/FNGameplayCues/Effects/Fort_Effects/Textures/HidingProp_Teleport/t_hideable_warp_light_masks
  }
  n19 = VectorParameter() {group=Emissive, parameterName=Emissive Light Color, defaultValue=<rgba 1 1 1 0>}
  n20 = VectorParameter() {group=Emissive, parameterName=Emissive Light Color 2, defaultValue=<rgba 1 1 1 0>}
  n21 = TextureSample(UVs=@n2) {samplerSource=SSM_Wrap_WorldGroupSettings, texture=/Game/Environments/Apollo/Sets/Yacht/Textures/T_PipeInterior_D}
  n22 = StaticBool()
  n23 = TextureSample!(UVs=Add(A=@n2, B=Constant2Vector() {r=-0.0025}), Tex, Bias=Constant() {r=1}, Apply View MipBias=@n22) {mipValueMode=TMVM_MipBias, samplerSource=SSM_Wrap_WorldGroupSettings, texture=&T_PipeInterior_D}
  n24 = TextureSample(UVs=Add(A=@n2, B=Constant2Vector() {r=-0.005}), Bias=Constant() {r=2}, Apply View MipBias=@n22) {mipValueMode=TMVM_MipBias, samplerSource=SSM_Wrap_WorldGroupSettings, texture=&T_PipeInterior_D}
  n25 = TextureSample(UVs=Add(A=@n2, B=Constant2Vector() {r=-0.0075}), Bias=Constant() {r=2}, Apply View MipBias=@n22) {mipValueMode=TMVM_MipBias, samplerSource=SSM_Wrap_WorldGroupSettings, texture=&T_PipeInterior_D}
  n26 = Add(A=Add(A=Add(A=@n21.RGB, B=@n23.RGB), B=@n24.RGB), B=@n25.RGB)
  n27 = QualitySwitch(Default=ShadingPathSwitch(Default=Divide(A=@n26) {constB=4}, Mobile=@n21.RGB) {desc=Migrated from FeatureLevelSwitch}, Low=@n21.RGB)
  n28 = TextureSampleParameterCube!(UVs=Transform(None=@n12), Apply View MipBias) {
   group=Cubemap
   parameterName=Cubemap Texture
   samplerSource=SSM_Wrap_WorldGroupSettings
   texture=/Game/Packages/Fortress_Sky/SkyDome/MaterialInstances/SkyDome_Apollo/Cubemap_Apollo_Night
  }
  n29 = Desaturation(None=@n28.RGB, Fraction=ScalarParameter() {defaultValue=1, group=Cubemap, parameterName=Cubemap Desaturation})
  n30 = VectorParameter() {group=Cubemap, parameterName=Cubemap Multiplier, defaultValue=<rgba 1 1 1 1>}
  n31 = ShadingPathSwitch(Default=Add(A=@n27, B=Multiply(A=@n29, B=@n30.RGB)), Mobile=@n27) {desc=Migrated from FeatureLevelSwitch}
  n32 = Constant() {r=4.7125}
  n33 = Rotator(
   Coordinate=MaterialFunctionCall(
    UV Channel=TextureCoordinate()
    U Freq Multiplier=ScalarParameter() {defaultValue=2, group=UVs, parameterName=UV SpeedLines U Mult}
    V Freq Multiplier=ScalarParameter() {defaultValue=0.05, group=UVs, parameterName=UV SpeedLines V Mult}
    Pan Speed U=ScalarParameter() {defaultValue=0.4, group=UVs, parameterName=UV SpeedLines U Pan Speed}
    Pan Speed V=ScalarParameter() {defaultValue=-0.25, group=UVs, parameterName=UV SpeedLines V Pan Speed}
   ) {materialFunction=&MF_FX_TextureUVPanner}.Result
   Time=@n32
  ) {speed=1}
  n34 = Multiply(
   A=Power(
    Base=TextureSample!(UVs=@n33, Tex, Apply View MipBias) {samplerSource=SSM_Wrap_WorldGroupSettings, texture=/Game/Global/Textures/Default/Noise/Noise}.G
    Exp=ScalarParameter() {defaultValue=12, parameterName=Speed Lines Power}
   ) {constExponent=18}
   B=VectorParameter() {parameterName=Speed Lines Color, defaultValue=<rgba 1 1 1 0>}.RGB
  ) {constB=0.25}
  n35 = TextureCoordinate()
  n36 = ScalarParameter() {defaultValue=2, group=UVs, parameterName=UV Mist U Mult}
  n37 = ScalarParameter() {defaultValue=3, group=UVs, parameterName=UV Mist V Mult}
  n38 = ScalarParameter() {defaultValue=0.4, group=UVs, parameterName=UV Mist U Pan Speed}
  n39 = ScalarParameter() {defaultValue=-1, group=UVs, parameterName=UV Mist V Pan Speed}
  n40 = TextureSample(UVs=Rotator(
    Coordinate=MaterialFunctionCall(UV Channel=@n35, U Freq Multiplier=@n36, V Freq Multiplier=@n37, Pan Speed U=@n38, Pan Speed V=@n39) {materialFunction=&MF_FX_TextureUVPanner}.Result
    Time=@n32
   ) {speed=1}) {samplerSource=SSM_Wrap_WorldGroupSettings, samplerType=SAMPLERTYPE_Grayscale, texture=&LowResBlurredNoise}
  n41 = Multiply(A=Power(Base=@n40.R) {constExponent=4}, B=VectorParameter() {parameterName=Mist Color, defaultValue=<rgba 1 1 1 1>}.RGB)
  n42 = Add(
   A=Multiply(A=@n18.G, B=LinearInterpolate(A=@n19.RGB, B=@n20.RGB, Alpha=@n18.B))
   B=Add(A=QualitySwitch(Default=@n31, Low=@n27), B=Add(A=@n34, B=@n41))
  )
  n43 = VectorParameter() {parameterName=Tunnel Mist Color, defaultValue=<rgba 0 1 1 0>}
  n44 = MaterialFunctionCall() {materialFunction=&LinearGradient}
  n45 = Power(Base=@n44.VGradient, Exp=ScalarParameter() {defaultValue=4, parameterName=GradientPowerInnerShadow})
  n46 = Multiply(
   A=Multiply(A=Multiply(A=@n13, B=@n42), B=LinearInterpolate(A=@n43.RGB, Alpha=@n45))
   B=Power(Base=OneMinus(None=@n44.VGradient), Exp=ScalarParameter() {defaultValue=0.2, parameterName=GradientPowerOuterShadow})
  )
  n47 = OneMinus(None=@n45)
  n48 = DynamicParameter() {paramNames [WhiteWash, Final Emissive Multiplier, Param3, Param4]}
  n49 = LinearInterpolate(A=Constant() {r=-1}, B=Constant() {r=1}, Alpha=@n48.WhiteWash)
  n50 = Clamp(None=Multiply(A=Add(A=@n47, B=@n49), B=VectorParameter() {parameterName=TunnelExitColor, defaultValue=<rgba 1 1 1 0>}.RGB))
  MaterialFunctionCall(Emissive=Multiply(A=Add(A=@n46, B=@n50), B=@n48.Final Emissive Multiplier)) {materialFunction=&MF_FX_EmissiveAdjustments}
  ParticleColor()
  n51 = Multiply(A=Time(), B=ScalarParameter() {defaultValue=0.25, parameterName=RotationSpeed})
  n52 = Add(A=@n51, B=Sine(None=Multiply(A=@n51) {constB=0.45}))
  n53 = Multiply(
   A=MaterialFunctionCall(R=Sine(None=@n52), B=Cosine(None=@n52)) {materialFunction=&AppendMany}.RGB
   B=ScalarParameter() {defaultValue=150, parameterName=Multiplier}
  )
  n54 = Distance(
   A=WorldPosition() {worldPositionShaderOffset=WPT_ExcludeAllShaderOffsets}.XYZ
   B=TransformPosition(None=VectorParameter() {parameterName=Offset, defaultValue=<rgba 0 50 0 1>}.RGB) {transformType=TRANSFORMPOSSOURCE_World}
  )
  n55 = ScalarParameter() {defaultValue=150, parameterName=Rad}
  n56 = OneMinus(None=Saturate(None=Divide(A=@n54, B=@n55)))
  n57 = ScalarParameter() {defaultValue=2, parameterName=LinearFalloffPower}
  Reroute(None=LinearInterpolate(A=Constant3Vector(), B=Transform(None=@n53) {transformSourceType=TRANSFORMSOURCE_Local}, Alpha=Power(Base=@n56, Exp=@n57)))
 ]
}
